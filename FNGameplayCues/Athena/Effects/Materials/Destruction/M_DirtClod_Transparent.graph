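# Translucent lit material for dirt clod sprite particles in Cascade or Niagara, in the Athena destruction effects folder.
# A dirt clod cluster texture is sampled with bump-offset UVs driven by its own alpha, then tinted by particle color.
# A UseDynamicParameter scalar blends between two alpha erosion methods: particle alpha minus inverted texture alpha, or multiplied alphas minus a dynamic Erosion parameter.
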Material /FNGameplayCues/Athena/Effects/Materials/Destruction/M_DirtClod_Transparent
graph {
 properties {
  bUsedWithNiagaraSprites=true
  bUsedWithParticleSprites=true
  blendMode=BLEND_Translucent
  pixelDepthOffsetMode=PDOM_Legacy
  usedShadingModels=Default Lit
 }
 nodes [
  n1 = ParticleColor()
  n2 = TextureSample(UVs=BumpOffset(Height=TextureSample() {texture=/Game/Effects/Fort_Effects/Textures/Dirt/T-DirtClodCluster_D}.A)) {texture=&T-DirtClodCluster_D}
  LinearInterpolate(
   A=Subtract(A=@n1.A, B=OneMinus(None=@n2.A))
   B=Subtract(A=Multiply(A=@n1.A, B=@n2.A), B=DynamicParameter() {defaultValue=<rgba 0 1 1 1>, paramNames [Erosion,Param2,Param3,Param4]}.Erosion) {constB=0.5}
   Alpha=ScalarParameter() {parameterName=UseDynamicParameter}
  )
  Reroute(None=Multiply(A=@n1.RGB, B=@n2.RGB))
 ]
}
